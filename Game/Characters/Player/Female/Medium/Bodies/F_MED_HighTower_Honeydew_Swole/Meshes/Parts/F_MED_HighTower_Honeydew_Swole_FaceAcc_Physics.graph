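# Physics asset for the FaceAcc mesh part of the female HighTower Honeydew Swole character, setting up dynamic hair physics against the upper body.
# Capsules cover head, neck, upper spine, clavicles and upper arms, joined by angularly free constraints.
# Center, left and right chains of box bodies hang from head-attached base capsules, with swing limits widening toward the tips.
# Cross constraints tie side chains to the center, and six small tail capsules branch from the chain ends.
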PhysicsAsset /Game/Characters/Player/Female/Medium/Bodies/F_MED_HighTower_Honeydew_Swole/Meshes/Parts/F_MED_HighTower_Honeydew_Swole_FaceAcc_Physics
bodies {
 clavicle_l [{shapeType=Capsule, length=13.125, radius=8.25, center=<7.171 -2.039 -1.9291>, rotation=<rot 83.166 0 -3.0997>}]
 clavicle_r [{shapeType=Capsule, length=17.25, radius=8, center=<-7.171 1.7746 1.9291>, rotation=<rot 83.1664 0 180>}]
 dyn_hairTail_a [{shapeType=Capsule, length=3.6121, radius=0.8091, center=<2.5342 -0.0336 0.24>, rotation=<rot 0.0127 -97.3482 -89.9027>}]
 dyn_hairTail_b [{shapeType=Capsule, length=2.5102, radius=0.8114, center=<2.1314 0.0622 0.1064>, rotation=<rot 0 88.1148 -84.9415>}]
 dyn_hairTail_c [{shapeType=Capsule, length=2.8766, radius=0.8598, center=<1.7234 -0.0116 -0.1993>, rotation=<rot 0 -84.2742 -87.637>}]
 dyn_hairTail_d [{shapeType=Capsule, length=2.8437, radius=0.6878, center=<1.8143 -0.0344 0.1887>, rotation=<rot 0 85.4957 -75.7154>}]
 dyn_hairTail_e [{shapeType=Capsule, length=4.0544, radius=0.7128, center=<2.7219 -0.2603 0.0448>, rotation=<rot 0 -82.4906 -85.7305>}]
 dyn_hairTail_f [{shapeType=Capsule, length=2.7299, radius=0.75, center=<1.774 -0.2256 0.0673>, rotation=<rot 0 87.2378 -87.3275>}]
 dyn_hair_main_a [{shapeType=Box, extentX=9, extentY=5, extentZ=15, center=<4.7037 0 0>}]
 dyn_hair_main_a_base [{shapeType=Capsule, length=1, radius=1}]
 dyn_hair_main_a_base_root_base_r [{shapeType=Capsule, length=1, radius=1}]
 dyn_hair_main_a_l [{shapeType=Box, extentX=8.5, extentY=3, extentZ=13.5, center=<4.9954 -0.0001 0>, rotation=<rot 0 0 -1.9411>}]
 dyn_hair_main_a_r [{shapeType=Box, extentX=8.5, extentY=3, extentZ=13.5, center=<-4.9954 0.0001 0>, rotation=<rot 0 0 178.0589>}]
 dyn_hair_main_a_root_base_l [{shapeType=Capsule, length=1, radius=1}]
 dyn_hair_main_a_root_l [{shapeType=Box, extentX=6.5, extentY=2.5, extentZ=13, center=<3.0047 0 0>}]
 dyn_hair_main_a_root_r [{shapeType=Box, extentX=6.5, extentY=2.5, extentZ=13, center=<-3.0047 0 0>, rotation=<rot 0 0 180>}]
 dyn_hair_main_b [{shapeType=Box, extentX=7.5, extentY=3, extentZ=11.5, center=<3.9414 0 0>}]
 dyn_hair_main_b_l [{shapeType=Box, extentX=9.5, extentY=15.5, extentZ=4.5, center=<5.3478 0 0>, rotation=<rot 0.0001 0 32.306>}]
 dyn_hair_main_b_r [{shapeType=Box, extentX=8, extentY=15, extentZ=3.5, center=<-5.3478 0 0>, rotation=<rot 0.0001 0.0001 -147.6938>}]
 dyn_hair_main_c [{shapeType=Box, extentX=8, extentY=2.5, extentZ=10, center=<3.7392 0 0>}]
 dyn_hair_main_c_l [{shapeType=Box, extentX=8.5, extentY=10.5, extentZ=2.5, center=<4.0487 0 -0.0001>}]
 dyn_hair_main_c_r [{shapeType=Box, extentX=7.5, extentY=10.5, extentZ=2.5, center=<-4.0487 0 0.0001>, rotation=<rot 0 0 180>}]
 dyn_hair_main_d [{shapeType=Box, extentX=11, extentY=2.5, extentZ=7, center=<5.3624 0 0>}]
 dyn_hair_main_d_l [{shapeType=Box, extentX=11, extentY=7, extentZ=2, center=<5.8092 0 0.0001>}]
 dyn_hair_main_d_r [{shapeType=Box, extentX=11, extentY=7, extentZ=2, center=<-5.8092 0 -0.0001>, rotation=<rot 0 0 180>}]
 head [{shapeType=Capsule, length=9.5, radius=6.25, center=<3.1009 1.7977 -0.0001>, rotation=<rot 90 7.125 1.2091>}]
 neck_01 [{shapeType=Capsule, length=13.125, radius=5.75, center=<3.656 0.3268 0>, rotation=<rot 90 0 -6.183>}]
 spine_03 [{shapeType=Capsule, length=20, radius=8.5, center=<1.0513 1.4965 0>}]
 spine_04 [{shapeType=Capsule, length=10.75, radius=10.25, center=<3.0449 -0.3824 -0.0001>}]
 spine_05 [{shapeType=Capsule, length=22.625, radius=11.25, center=<3.5387 0.086 -0.0001>}]
 upperarm_l [{shapeType=Capsule, length=16.75, radius=8.375, center=<10.6387 -1.5356 -0.0913>, rotation=<rot 86.0246 0 0>}]
 upperarm_r [{shapeType=Capsule, length=20.625, radius=6.25, center=<-10.6387 1.5356 0.0913>, rotation=<rot 86.0246 0 -180>}]
}
constraints [
 {
  bone1Name=head
  bone2Name=neck_01
  swing1LimitDegrees=45
  swing1Motion=Free
  swing2LimitDegrees=45
  swing2Motion=Free
  twistLimitDegrees=45
  twistMotion=Free
 }
 {
  bone1Name=neck_01
  bone2Name=spine_05
  swing1LimitDegrees=45
  swing1Motion=Free
  swing2LimitDegrees=45
  swing2Motion=Free
  twistLimitDegrees=45
  twistMotion=Free
 }
 {
  bone1Name=spine_05
  bone2Name=spine_04
  swing1LimitDegrees=45
  swing1Motion=Free
  swing2LimitDegrees=45
  swing2Motion=Free
  twistLimitDegrees=45
  twistMotion=Free
 }
 {
  bone1Name=spine_04
  bone2Name=spine_03
  swing1LimitDegrees=45
  swing1Motion=Free
  swing2LimitDegrees=45
  swing2Motion=Free
  twistLimitDegrees=45
  twistMotion=Free
 }
 {
  bone1Name=clavicle_l
  bone2Name=spine_05
  swing1LimitDegrees=45
  swing1Motion=Free
  swing2LimitDegrees=45
  swing2Motion=Free
  twistLimitDegrees=45
  twistMotion=Free
 }
 {
  bone1Name=upperarm_l
  bone2Name=clavicle_l
  swing1LimitDegrees=45
  swing1Motion=Free
  swing2LimitDegrees=45
  swing2Motion=Free
  twistLimitDegrees=45
  twistMotion=Free
 }
 {
  bone1Name=clavicle_r
  bone2Name=spine_05
  swing1LimitDegrees=45
  swing1Motion=Free
  swing2LimitDegrees=45
  swing2Motion=Free
  twistLimitDegrees=45
  twistMotion=Free
 }
 {
  bone1Name=upperarm_r
  bone2Name=clavicle_r
  swing1LimitDegrees=45
  swing1Motion=Free
  swing2LimitDegrees=45
  swing2Motion=Free
  twistLimitDegrees=45
  twistMotion=Free
 }
 {
  bone1Name=dyn_hair_main_a_base
  bone2Name=head
  swing1LimitDegrees=45
  swing1Motion=Free
  swing2LimitDegrees=45
  swing2Motion=Free
  twistLimitDegrees=45
  twistMotion=Free
 }
 {
  bone1Name=dyn_hair_main_a_root_base_l
  bone2Name=head
  swing1LimitDegrees=45
  swing1Motion=Free
  swing2LimitDegrees=45
  swing2Motion=Free
  twistLimitDegrees=45
  twistMotion=Free
 }
 {
  bone1Name=dyn_hair_main_a_base_root_base_r
  bone2Name=head
  swing1LimitDegrees=45
  swing1Motion=Free
  swing2LimitDegrees=45
  swing2Motion=Free
  twistLimitDegrees=45
  twistMotion=Free
 }
 {bone1Name=dyn_hair_main_a, bone2Name=dyn_hair_main_a_base, swing1LimitDegrees=20, swing2LimitDegrees=20, twistLimitDegrees=5}
 {bone1Name=dyn_hair_main_b, bone2Name=dyn_hair_main_a, swing1LimitDegrees=45, swing2LimitDegrees=45, twistLimitDegrees=5}
 {bone1Name=dyn_hair_main_c, bone2Name=dyn_hair_main_b, swing1LimitDegrees=45, swing2LimitDegrees=45, twistLimitDegrees=5}
 {bone1Name=dyn_hair_main_d, bone2Name=dyn_hair_main_c, swing1LimitDegrees=45, swing2LimitDegrees=45, twistLimitDegrees=5}
 {bone1Name=dyn_hair_main_a_root_l, bone2Name=dyn_hair_main_a_root_base_l, swing1LimitDegrees=10, swing2LimitDegrees=10, twistLimitDegrees=2}
 {bone1Name=dyn_hair_main_a_l, bone2Name=dyn_hair_main_a_root_l, swing1LimitDegrees=20, swing2LimitDegrees=20, twistLimitDegrees=1}
 {bone1Name=dyn_hair_main_b_l, bone2Name=dyn_hair_main_a_l, swing1LimitDegrees=30, swing2LimitDegrees=30, twistLimitDegrees=2}
 {bone1Name=dyn_hair_main_c_l, bone2Name=dyn_hair_main_b_l, swing1LimitDegrees=45, swing2LimitDegrees=45, twistLimitDegrees=2}
 {bone1Name=dyn_hair_main_d_l, bone2Name=dyn_hair_main_c_l, swing1LimitDegrees=45, swing2LimitDegrees=45, twistLimitDegrees=5}
 {
  bone1Name=dyn_hair_main_a_root_r
  bone2Name=dyn_hair_main_a_base_root_base_r
  swing1LimitDegrees=10
  swing2LimitDegrees=10
  twistLimitDegrees=2
 }
 {bone1Name=dyn_hair_main_a_r, bone2Name=dyn_hair_main_a_root_r, swing1LimitDegrees=20, swing2LimitDegrees=20, twistLimitDegrees=2}
 {bone1Name=dyn_hair_main_b_r, bone2Name=dyn_hair_main_a_r, swing1LimitDegrees=30, swing2LimitDegrees=30, twistLimitDegrees=5}
 {bone1Name=dyn_hair_main_c_r, bone2Name=dyn_hair_main_b_r, swing1LimitDegrees=45, swing2LimitDegrees=45, twistLimitDegrees=5}
 {bone1Name=dyn_hair_main_d_r, bone2Name=dyn_hair_main_c_r, swing1LimitDegrees=45, swing2LimitDegrees=45, twistLimitDegrees=5}
 {bone1Name=dyn_hair_main_c_l, bone2Name=dyn_hair_main_c, swing1LimitDegrees=45, swing2LimitDegrees=45, twistLimitDegrees=45}
 {bone1Name=dyn_hair_main_c_r, bone2Name=dyn_hair_main_c, swing1LimitDegrees=45, swing2LimitDegrees=45, twistLimitDegrees=45}
 {bone1Name=dyn_hair_main_d_l, bone2Name=dyn_hair_main_d, swing1LimitDegrees=45, swing2LimitDegrees=45, twistLimitDegrees=45}
 {bone1Name=dyn_hair_main_d_r, bone2Name=dyn_hair_main_d, swing1LimitDegrees=45, swing2LimitDegrees=45, twistLimitDegrees=45}
 {bone1Name=dyn_hair_main_b_l, bone2Name=dyn_hair_main_b, swing1LimitDegrees=45, swing2LimitDegrees=45, twistLimitDegrees=45}
 {bone1Name=dyn_hair_main_b_r, bone2Name=dyn_hair_main_b, swing1LimitDegrees=45, swing2LimitDegrees=45, twistLimitDegrees=45}
 {bone1Name=dyn_hairTail_a, bone2Name=dyn_hair_main_d_l, swing1LimitDegrees=30, swing2LimitDegrees=30, twistLimitDegrees=5}
 {bone1Name=dyn_hairTail_b, bone2Name=dyn_hair_main_d_l, swing1LimitDegrees=30, swing2LimitDegrees=30, twistLimitDegrees=5}
 {bone1Name=dyn_hairTail_c, bone2Name=dyn_hair_main_d, swing1LimitDegrees=30, swing2LimitDegrees=30, twistLimitDegrees=5}
 {bone1Name=dyn_hairTail_d, bone2Name=dyn_hair_main_d, swing1LimitDegrees=30, swing2LimitDegrees=30, twistLimitDegrees=5}
 {bone1Name=dyn_hairTail_e, bone2Name=dyn_hair_main_d, swing1LimitDegrees=30, swing2LimitDegrees=30, twistLimitDegrees=5}
 {bone1Name=dyn_hairTail_f, bone2Name=dyn_hair_main_d_r, swing1LimitDegrees=30, swing2LimitDegrees=30, twistLimitDegrees=5}
]
