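# Physics asset for the HairSec (hair) part of the female medium Honeydew Swole player body, using only capsule bodies.
# Upper spine, neck, head, clavicle and upper-arm capsules form the torso hierarchy, with main and base hair bodies branching from the head through free-rotation constraints.
# Short strand chains hang off those with tight swing and twist limits: A through F on the left, only A on the right.
# The left D chain locks one swing axis.
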PhysicsAsset /Game/Characters/Player/Female/Medium/Bodies/F_MED_HighTower_Honeydew_Swole/Meshes/Parts/F_MED_Honeydew_Swole_HairSec_Physics
bodies {
 clavicle_l [{shapeType=Capsule, length=4.5, radius=8.25, center=<7.1468 -1.7746 -2.1314>, rotation=<rot 83.1664 0 0>}]
 clavicle_r [{shapeType=Capsule, length=4.75, radius=8, center=<-7.171 1.7746 1.9291>, rotation=<rot 83.1664 0 180>}]
 dyn_hair_A_a_base_l [{shapeType=Capsule, length=1, radius=1}]
 dyn_hair_A_a_base_r [{shapeType=Capsule, length=1, radius=1}]
 dyn_hair_A_a_l [{shapeType=Capsule, length=2.25, radius=1.5, center=<2.7921 0 0>, rotation=<rot 90 0 0>}]
 dyn_hair_A_a_r [{shapeType=Capsule, length=3.5, radius=2, center=<3.3415 0 0.0001>, rotation=<rot 90 0 0>}]
 dyn_hair_A_b_l [{shapeType=Capsule, length=2, radius=1.5, center=<1.8969 0 0>, rotation=<rot 90 0 0>}]
 dyn_hair_A_b_r [{shapeType=Capsule, length=4.25, radius=2.25, center=<4.3364 0.0001 -0.5478>, rotation=<rot 86.2512 0 0>}]
 dyn_hair_A_c_l [{shapeType=Capsule, length=2.5, radius=1.25, center=<2.2906 0 0>, rotation=<rot 90 0 0>}]
 dyn_hair_B_a_l [{shapeType=Capsule, length=2, radius=1.75, center=<2.6975 0.4396 0.3746>, rotation=<rot 82.0925 -179.9999 174.7474>}]
 dyn_hair_B_b_l [{shapeType=Capsule, length=3, radius=1.25, center=<2.5547 0 0>, rotation=<rot 90 0 0>}]
 dyn_hair_B_c_l [{shapeType=Capsule, length=3.5, radius=1, center=<2.4697 0 0.0001>, rotation=<rot 90 0 0>}]
 dyn_hair_B_d_l [{shapeType=Capsule, length=3.75, radius=1, center=<2.5243 0 0>, rotation=<rot 90 0 0>}]
 dyn_hair_C_a_l [{shapeType=Capsule, length=4, radius=1.5, center=<3.028 0 -0.0001>, rotation=<rot 90 0 0>}]
 dyn_hair_C_b_l [{shapeType=Capsule, length=4.25, radius=1, center=<3.2974 0 -0.0001>, rotation=<rot 90 0 0>}]
 dyn_hair_D_a_l [{shapeType=Capsule, length=2, radius=1.25, center=<2.2728 0 0>, rotation=<rot 90 0 0>}]
 dyn_hair_D_b_l [{shapeType=Capsule, length=4.5, radius=1, center=<3.5253 0 0>, rotation=<rot 90 0 0>}]
 dyn_hair_E_a_l [{shapeType=Capsule, length=5.25, radius=0.75, center=<3.3922 0 0>, rotation=<rot 90 0 0>}]
 dyn_hair_F_a_l [{shapeType=Capsule, length=2.5, radius=0.75, center=<1.0948 0 0>, rotation=<rot 90 0 0>}]
 dyn_hair_main_a_base [{shapeType=Capsule, length=1, radius=1}]
 dyn_hair_main_a_l [{shapeType=Capsule, length=1, radius=1}]
 dyn_hair_main_a_r [{shapeType=Capsule, length=2.25, radius=5, center=<-1.8534 -5.5449 3.0647>, rotation=<rot 78.8712 -3.3123 0>}]
 dyn_hair_main_a_root_l [{shapeType=Capsule, length=1, radius=1}]
 dyn_hair_main_b_l [{shapeType=Capsule, length=2.75, radius=3.75, center=<3.3989 1.4547 1.3762>, rotation=<rot 90 -90 -87.2979>}]
 dyn_hair_main_b_r [{shapeType=Capsule, length=9.75, radius=5.5, center=<-2.5598 -3.7835 1.8105>, rotation=<rot 90 0 0>}]
 dyn_hair_main_c_l [{shapeType=Capsule, length=2.25, radius=4.25, center=<0 1.0524 2.4506>, rotation=<rot 90 0 0>}]
 head [{shapeType=Capsule, length=6.25, radius=6.25, center=<3.1009 1.7977 -0.0001>, rotation=<rot 90 7.125 1.2091>}]
 neck_01 [{shapeType=Capsule, length=4.5, radius=5.25, center=<3.656 0.3268 0>, rotation=<rot 90 0 -6.183>}]
 spine_03 [{shapeType=Capsule, length=6.75, radius=8.5, center=<1.0513 1.4965 0>}]
 spine_04 [{shapeType=Capsule, length=10.75, radius=10.25, center=<3.0449 -0.3824 -0.0001>}]
 spine_05 [{shapeType=Capsule, length=7.5, radius=11.25, center=<2.5045 0.086 0>}]
 upperarm_l [{shapeType=Capsule, length=12.75, radius=6.5, center=<10.6387 -1.5356 -0.0913>, rotation=<rot 86.0246 0 0>}]
 upperarm_r [{shapeType=Capsule, length=13, radius=6.25, center=<-10.6387 1.5356 0.0913>, rotation=<rot 86.0246 0 -180>}]
}
constraints [
 {
  bone1Name=head
  bone2Name=neck_01
  swing1LimitDegrees=45
  swing1Motion=Free
  swing2LimitDegrees=45
  swing2Motion=Free
  twistLimitDegrees=45
  twistMotion=Free
 }
 {
  bone1Name=neck_01
  bone2Name=spine_05
  swing1LimitDegrees=45
  swing1Motion=Free
  swing2LimitDegrees=45
  swing2Motion=Free
  twistLimitDegrees=45
  twistMotion=Free
 }
 {
  bone1Name=spine_05
  bone2Name=spine_04
  swing1LimitDegrees=45
  swing1Motion=Free
  swing2LimitDegrees=45
  swing2Motion=Free
  twistLimitDegrees=45
  twistMotion=Free
 }
 {
  bone1Name=spine_04
  bone2Name=spine_03
  swing1LimitDegrees=45
  swing1Motion=Free
  swing2LimitDegrees=45
  swing2Motion=Free
  twistLimitDegrees=45
  twistMotion=Free
 }
 {
  bone1Name=clavicle_l
  bone2Name=spine_05
  swing1LimitDegrees=45
  swing1Motion=Free
  swing2LimitDegrees=45
  swing2Motion=Free
  twistLimitDegrees=45
  twistMotion=Free
 }
 {
  bone1Name=upperarm_l
  bone2Name=clavicle_l
  swing1LimitDegrees=45
  swing1Motion=Free
  swing2LimitDegrees=45
  swing2Motion=Free
  twistLimitDegrees=45
  twistMotion=Free
 }
 {
  bone1Name=clavicle_r
  bone2Name=spine_05
  swing1LimitDegrees=45
  swing1Motion=Free
  swing2LimitDegrees=45
  swing2Motion=Free
  twistLimitDegrees=45
  twistMotion=Free
 }
 {
  bone1Name=upperarm_r
  bone2Name=clavicle_r
  swing1LimitDegrees=45
  swing1Motion=Free
  swing2LimitDegrees=45
  swing2Motion=Free
  twistLimitDegrees=45
  twistMotion=Free
 }
 {
  bone1Name=dyn_hair_main_a_base
  bone2Name=head
  swing1LimitDegrees=45
  swing1Motion=Free
  swing2LimitDegrees=45
  swing2Motion=Free
  twistLimitDegrees=45
  twistMotion=Free
 }
 {
  bone1Name=dyn_hair_A_a_base_l
  bone2Name=dyn_hair_main_a_root_l
  swing1LimitDegrees=45
  swing1Motion=Free
  swing2LimitDegrees=45
  swing2Motion=Free
  twistLimitDegrees=45
  twistMotion=Free
 }
 {bone1Name=dyn_hair_A_a_l, bone2Name=dyn_hair_A_a_base_l, swing1LimitDegrees=15, swing2LimitDegrees=15, twistLimitDegrees=5}
 {bone1Name=dyn_hair_A_b_l, bone2Name=dyn_hair_A_a_l, swing1LimitDegrees=15, swing2LimitDegrees=15, twistLimitDegrees=5}
 {bone1Name=dyn_hair_A_c_l, bone2Name=dyn_hair_A_b_l, swing1LimitDegrees=15, swing2LimitDegrees=15, twistLimitDegrees=5}
 {
  bone1Name=dyn_hair_main_a_root_l
  bone2Name=head
  swing1LimitDegrees=45
  swing1Motion=Free
  swing2LimitDegrees=45
  swing2Motion=Free
  twistLimitDegrees=45
  twistMotion=Free
 }
 {
  bone1Name=dyn_hair_main_a_l
  bone2Name=dyn_hair_main_a_root_l
  swing1LimitDegrees=45
  swing1Motion=Free
  swing2LimitDegrees=45
  swing2Motion=Free
  twistLimitDegrees=45
  twistMotion=Free
 }
 {bone1Name=dyn_hair_B_a_l, bone2Name=dyn_hair_main_a_l, swing1LimitDegrees=5, swing2LimitDegrees=5, twistLimitDegrees=5}
 {bone1Name=dyn_hair_B_b_l, bone2Name=dyn_hair_B_a_l, swing1LimitDegrees=5, swing2LimitDegrees=5, twistLimitDegrees=5}
 {bone1Name=dyn_hair_B_c_l, bone2Name=dyn_hair_B_b_l, swing1LimitDegrees=5, swing2LimitDegrees=5, twistLimitDegrees=5}
 {bone1Name=dyn_hair_B_d_l, bone2Name=dyn_hair_B_c_l, swing1LimitDegrees=5, swing2LimitDegrees=5, twistLimitDegrees=5}
 {bone1Name=dyn_hair_C_a_l, bone2Name=dyn_hair_main_a_l, swing1LimitDegrees=10, swing2LimitDegrees=10, twistLimitDegrees=5}
 {bone1Name=dyn_hair_C_b_l, bone2Name=dyn_hair_C_a_l, swing1LimitDegrees=10, swing2LimitDegrees=10, twistLimitDegrees=5}
 {bone1Name=dyn_hair_D_a_l, bone2Name=dyn_hair_main_a_l, swing1LimitDegrees=5, swing1Motion=Locked, swing2LimitDegrees=10, twistLimitDegrees=5}
 {bone1Name=dyn_hair_D_b_l, bone2Name=dyn_hair_D_a_l, swing1LimitDegrees=5, swing1Motion=Locked, swing2LimitDegrees=10, twistLimitDegrees=5}
 {bone1Name=dyn_hair_E_a_l, bone2Name=dyn_hair_main_a_l, swing1LimitDegrees=15, swing2LimitDegrees=15, twistLimitDegrees=5}
 {
  bone1Name=dyn_hair_main_b_l
  bone2Name=dyn_hair_main_a_l
  swing1LimitDegrees=45
  swing1Motion=Free
  swing2LimitDegrees=45
  swing2Motion=Free
  twistLimitDegrees=45
  twistMotion=Free
 }
 {
  bone1Name=dyn_hair_main_c_l
  bone2Name=dyn_hair_main_b_l
  swing1LimitDegrees=45
  swing1Motion=Free
  swing2LimitDegrees=45
  swing2Motion=Free
  twistLimitDegrees=45
  twistMotion=Free
 }
 {bone1Name=dyn_hair_F_a_l, bone2Name=dyn_hair_main_b_l, swing1LimitDegrees=10, swing2LimitDegrees=10, twistLimitDegrees=0}
 {
  bone1Name=dyn_hair_A_a_base_r
  bone2Name=head
  swing1LimitDegrees=45
  swing1Motion=Free
  swing2LimitDegrees=45
  swing2Motion=Free
  twistLimitDegrees=45
  twistMotion=Free
 }
 {bone1Name=dyn_hair_A_a_r, bone2Name=dyn_hair_A_a_base_r, swing1LimitDegrees=15, swing2LimitDegrees=30, twistLimitDegrees=5}
 {bone1Name=dyn_hair_A_b_r, bone2Name=dyn_hair_A_a_r, swing1LimitDegrees=10, swing2LimitDegrees=30, twistLimitDegrees=5}
 {
  bone1Name=dyn_hair_main_a_r
  bone2Name=head
  swing1LimitDegrees=45
  swing1Motion=Free
  swing2LimitDegrees=45
  swing2Motion=Free
  twistLimitDegrees=45
  twistMotion=Free
 }
 {
  bone1Name=dyn_hair_main_b_r
  bone2Name=dyn_hair_main_a_r
  swing1LimitDegrees=45
  swing1Motion=Free
  swing2LimitDegrees=45
  swing2Motion=Free
  twistLimitDegrees=45
  twistMotion=Free
 }
]
