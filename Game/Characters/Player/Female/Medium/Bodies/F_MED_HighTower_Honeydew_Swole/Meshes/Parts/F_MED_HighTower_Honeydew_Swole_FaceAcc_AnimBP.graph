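# Post-copy animation Blueprint for the Honeydew Swole face accessory mesh, adding secondary motion to its hair bones.
# It copies the attached parent's pose, layers Body_Mesh_b's upper arms over it, offsets hair base bones by HeadRotation, then caches it as 'Sim'.
# bRigidBodyEnabled selects either two chained RigidBody simulations or a fallback of bone constraints, fixed offsets and per-bone spring AnimDynamics, re-constrained to the clavicles.
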AnimBlueprint /Game/Characters/Player/Female/Medium/Bodies/F_MED_HighTower_Honeydew_Swole/Meshes/Parts/F_MED_HighTower_Honeydew_Swole_FaceAcc_AnimBP
target_skeleton=&/F_MED_HighTower_Honeydew_Swole_FaceAcc_Skeleton
graphs [
 AnimGraph {nodes [
   Local To Component(LocalPose=Copy Pose From Mesh() {bUseAttachedParent=true})
   "Transform (Modify) Bone - Bone: dyn_hair_main_b_r"!(ComponentPose, Alpha=1, Translation="0.000,0.000,1.5", Rotation="0,0,0") -> (Pose) {
    rotationMode=BMM_Additive
    rotationSpace=BCS_BoneSpace
    translationMode=BMM_Additive
    translationSpace=BCS_BoneSpace
    alphaBoolBlend {blendOption=Linear}
    alphaScaleBiasClamp {inRange {max=0, min=-25}, outRange {max=0, min=1}}
    boneToModify {boneName=dyn_hair_main_b_r}
    translation {z=1.5}
   }
   n1 = RigidBody(ComponentPose=Local To Component(LocalPose=Use cached pose 'Sim'())) {
    bClampLinearTranslationLimitToRefPose=true
    bOverrideWorldGravity=true
    overridePhysicsAsset=&/F_MED_HighTower_Honeydew_Swole_FaceAcc_Physics
    simulationSpace=BaseBoneSpace
    alphaBoolBlend {blendOption=Linear}
    baseBoneRef {boneName=spine_04}
    componentLinearAccScale=<0.4 0.4 0.4>
    componentLinearVelScale=<0.4 0.4 0.4>
    overrideWorldGravity {z=-980}
   }
   n2 = RigidBody(ComponentPose=@n1) {
    bClampLinearTranslationLimitToRefPose=true
    bOverrideWorldGravity=true
    overridePhysicsAsset=&/F_MED_Honeydew_Swole_HairSec_Physics
    simulationSpace=BaseBoneSpace
    alphaBoolBlend {blendOption=Linear}
    baseBoneRef {boneName=spine_04}
    componentLinearAccScale=<0.35 0.35 0.35>
    componentLinearVelScale=<0.35 0.35 0.35>
    overrideWorldGravity {z=-980}
   }
   n3 = "Constraint - dyn_hair_main_a (P)"(ComponentPose=Local To Component(LocalPose=Use cached pose 'Sim'()), Alpha=1, ConstraintWeights_0=1) {
    alphaBoolBlend {blendOption=Linear}
    boneToModify {boneName=dyn_hair_main_a}
    constraintSetup [{targetBone {boneName=neck_02}}]
    constraintWeights [1]
   }
   n4 = "Constraint - dyn_hair_main_b_l (P)"(ComponentPose=@n3, Alpha=0.2, ConstraintWeights_0=1) {
    alphaBoolBlend {blendOption=Linear}
    boneToModify {boneName=dyn_hair_main_b_l}
    constraintSetup [{targetBone {boneName=neck_02}}]
    constraintWeights [1]
   }
   n5 = "Constraint - dyn_hair_main_b_r (P)"(ComponentPose=@n4, Alpha=0.2, ConstraintWeights_0=1) {
    alphaBoolBlend {blendOption=Linear}
    boneToModify {boneName=dyn_hair_main_b_r}
    constraintSetup [{targetBone {boneName=clavicle_r}}]
    constraintWeights [1]
   }
   n6 = "Constraint - dyn_hair_main_c (P)"(ComponentPose=@n5, Alpha=0.8, ConstraintWeights_0=1) {
    alphaBoolBlend {blendOption=Linear}
    boneToModify {boneName=dyn_hair_main_c}
    constraintSetup [{targetBone {boneName=spine_05}}]
    constraintWeights [1]
   }
   n7 = "Constraint - dyn_hair_main_c_l (P)"(ComponentPose=@n6, Alpha=0.8, ConstraintWeights_0=1) {
    alphaBoolBlend {blendOption=Linear}
    boneToModify {boneName=dyn_hair_main_c_l}
    constraintSetup [{targetBone {boneName=spine_05}}]
    constraintWeights [1]
   }
   n8 = "Constraint - dyn_hair_main_c_r (P)"(ComponentPose=@n7, Alpha=0.8, ConstraintWeights_0=1) {
    alphaBoolBlend {blendOption=Linear}
    boneToModify {boneName=dyn_hair_main_c_r}
    constraintSetup [{targetBone {boneName=spine_05}}]
    constraintWeights [1]
   }
   n9 = "Transform (Modify) Bone - Bone: dyn_hair_main_a_base"!(ComponentPose=@n8, Alpha=1, Translation="0,0,0", Rotation="0.000,-35.000,0.000") -> (Pose) {
    rotationMode=BMM_Additive
    rotationSpace=BCS_BoneSpace
    translationMode=BMM_Additive
    translationSpace=BCS_BoneSpace
    alphaBoolBlend {blendOption=Linear}
    alphaScaleBiasClamp {inRange {max=0, min=-25}, outRange {max=0, min=1}}
    boneToModify {boneName=dyn_hair_main_a_base}
    rotation {yaw=-35.0001}
   }
   n10 = "Transform (Modify) Bone - Bone: dyn_hair_main_a_root_base_l"!(ComponentPose=@n9, Alpha=1, Translation="0,0,0", Rotation="-10.000,-5,-0.000") -> (Pose) {
    rotationMode=BMM_Additive
    rotationSpace=BCS_BoneSpace
    translationMode=BMM_Additive
    translationSpace=BCS_BoneSpace
    alphaBoolBlend {blendOption=Linear}
    alphaScaleBiasClamp {inRange {max=0, min=-25}, outRange {max=0, min=1}}
    boneToModify {boneName=dyn_hair_main_a_root_base_l}
    rotation {pitch=-10, yaw=-5}
   }
   n11 = "Transform (Modify) Bone - Bone: dyn_hair_main_a_base_root_base_r"!(ComponentPose=@n10, Alpha=1, Translation="0,0,0", Rotation="-10.000,-5,-0.000") -> (Pose) {
    rotationMode=BMM_Additive
    rotationSpace=BCS_BoneSpace
    translationMode=BMM_Additive
    translationSpace=BCS_BoneSpace
    alphaBoolBlend {blendOption=Linear}
    alphaScaleBiasClamp {inRange {max=0, min=-25}, outRange {max=0, min=1}}
    boneToModify {boneName=dyn_hair_main_a_base_root_base_r}
    rotation {pitch=-10, yaw=-5}
   }
   n12 = "Transform (Modify) Bone - Bone: dyn_hair_main_c_l"!(ComponentPose=@n11, Alpha=1, Translation="0,0,0", Rotation="-20,0,0") -> (Pose) {
    rotationMode=BMM_Additive
    rotationSpace=BCS_BoneSpace
    translationMode=BMM_Additive
    translationSpace=BCS_BoneSpace
    alphaBoolBlend {blendOption=Linear}
    alphaScaleBiasClamp {inRange {max=0, min=-25}, outRange {max=0, min=1}}
    boneToModify {boneName=dyn_hair_main_c_l}
    rotation {pitch=-20}
   }
   n13 = Anim Dynamics - Bone: dyn_hair_main_a(ComponentPose=@n12, Alpha=1, ExternalForce="0,0,0") {
    angularDampingOverride=0.9999
    angularSpringConstant=100
    bAngularSpring=true
    bGravityOverrideInSimSpace=true
    bOverrideAngularDamping=true
    simulationSpace=RootRelative
    alphaBoolBlend {blendOption=Linear}
    boundBone {boneName=dyn_hair_main_a}
    componentAppliedLinearAccClamp=<3500 3500 3500>
    componentLinearAccScale=<1.2 1.2 1.2>
    componentLinearVelScale=<1.2 1.2 1.2>
    physicsBodyDefinitions [{
      sphereCollisionRadius=0
      boundBone {boneName=dyn_hair_main_a}
      boxExtents=<20 10 10>
      constraintSetup {angularLimitsMax=<1 5 2>, angularLimitsMin=<-1 -5 -8>, angularTarget=<0 0 0>}
      localJointOffset=<2 0 0>
     }]
    relativeSpaceBone {boneName=root}
   }
   n14 = Anim Dynamics - Bone: dyn_hair_main_b(ComponentPose=@n13, Alpha=1, ExternalForce="0,0,0") {
    angularDampingOverride=0.9999
    angularSpringConstant=100
    bAngularSpring=true
    bGravityOverrideInSimSpace=true
    bOverrideAngularDamping=true
    simulationSpace=RootRelative
    alphaBoolBlend {blendOption=Linear}
    boundBone {boneName=dyn_hair_main_b}
    componentAppliedLinearAccClamp=<3500 3500 3500>
    componentLinearAccScale=<1.2 1.2 1.2>
    componentLinearVelScale=<1.2 1.2 1.2>
    physicsBodyDefinitions [{
      sphereCollisionRadius=0
      boundBone {boneName=dyn_hair_main_b}
      boxExtents=<20 10 10>
      constraintSetup {angularLimitsMax=<1 5 2>, angularLimitsMin=<-1 -5 -8>, angularTarget=<0 0 0>}
      localJointOffset=<2 0 0>
     }]
    relativeSpaceBone {boneName=root}
   }
   n15 = Anim Dynamics - Bone: dyn_hair_main_c(ComponentPose=@n14, Alpha=1, ExternalForce="0,0,0") {
    angularDampingOverride=0.9999
    angularSpringConstant=100
    bAngularSpring=true
    bGravityOverrideInSimSpace=true
    bOverrideAngularDamping=true
    simulationSpace=RootRelative
    alphaBoolBlend {blendOption=Linear}
    boundBone {boneName=dyn_hair_main_c}
    componentAppliedLinearAccClamp=<3500 3500 3500>
    componentLinearAccScale=<1.2 1.2 1.2>
    componentLinearVelScale=<1.2 1.2 1.2>
    physicsBodyDefinitions [{
      sphereCollisionRadius=0
      boundBone {boneName=dyn_hair_main_c}
      boxExtents=<20 10 10>
      constraintSetup {angularLimitsMax=<1 5 2>, angularLimitsMin=<-1 -5 -8>, angularTarget=<0 0 0>}
      localJointOffset=<2 0 0>
     }]
    relativeSpaceBone {boneName=root}
   }
   n16 = Anim Dynamics - Bone: dyn_hair_main_a_l(ComponentPose=@n15, Alpha=1, ExternalForce="0,0,0") {
    angularDampingOverride=0.9999
    angularSpringConstant=100
    bAngularSpring=true
    bGravityOverrideInSimSpace=true
    bOverrideAngularDamping=true
    simulationSpace=RootRelative
    alphaBoolBlend {blendOption=Linear}
    boundBone {boneName=dyn_hair_main_a_l}
    componentAppliedLinearAccClamp=<3500 3500 3500>
    componentLinearAccScale=<1.2 1.2 1.2>
    componentLinearVelScale=<1.2 1.2 1.2>
    physicsBodyDefinitions [{
      sphereCollisionRadius=0
      boundBone {boneName=dyn_hair_main_a_l}
      boxExtents=<20 10 10>
      constraintSetup {angularLimitsMax=<1 10 2>, angularLimitsMin=<-1 -2 -15>, angularTarget=<0 0 0>}
      localJointOffset=<2 0 0>
     }]
    relativeSpaceBone {boneName=root}
   }
   n17 = Anim Dynamics - Bone: dyn_hair_main_b_l(ComponentPose=@n16, Alpha=1, ExternalForce="0,0,0") {
    angularDampingOverride=0.9999
    angularSpringConstant=100
    bAngularSpring=true
    bGravityOverrideInSimSpace=true
    bOverrideAngularDamping=true
    simulationSpace=RootRelative
    alphaBoolBlend {blendOption=Linear}
    boundBone {boneName=dyn_hair_main_b_l}
    componentAppliedLinearAccClamp=<3500 3500 3500>
    componentLinearAccScale=<1.2 1.2 1.2>
    componentLinearVelScale=<1.2 1.2 1.2>
    physicsBodyDefinitions [{
      sphereCollisionRadius=0
      boundBone {boneName=dyn_hair_main_b_l}
      boxExtents=<20 10 10>
      constraintSetup {angularLimitsMax=<1 10 2>, angularLimitsMin=<-1 -2 -15>, angularTarget=<0 0 0>}
      localJointOffset=<2 0 0>
     }]
    relativeSpaceBone {boneName=root}
   }
   n18 = Anim Dynamics - Bone: dyn_hair_main_c_l(ComponentPose=@n17, Alpha=1, ExternalForce="0,0,0") {
    angularDampingOverride=0.9999
    angularSpringConstant=100
    bAngularSpring=true
    bGravityOverrideInSimSpace=true
    bOverrideAngularDamping=true
    simulationSpace=RootRelative
    alphaBoolBlend {blendOption=Linear}
    boundBone {boneName=dyn_hair_main_c_l}
    componentAppliedLinearAccClamp=<3500 3500 3500>
    componentLinearAccScale=<1.2 1.2 1.2>
    componentLinearVelScale=<1.2 1.2 1.2>
    physicsBodyDefinitions [{
      sphereCollisionRadius=0
      boundBone {boneName=dyn_hair_main_c_l}
      boxExtents=<20 10 10>
      constraintSetup {angularLimitsMax=<1 10 2>, angularLimitsMin=<-1 -2 -15>, angularTarget=<0 0 0>}
      localJointOffset=<2 0 0>
     }]
    relativeSpaceBone {boneName=root}
   }
   n19 = Anim Dynamics - Bone: dyn_hair_main_a_r(ComponentPose=@n18, Alpha=1, ExternalForce="0,0,0") {
    angularDampingOverride=0.9999
    angularSpringConstant=100
    bAngularSpring=true
    bGravityOverrideInSimSpace=true
    bOverrideAngularDamping=true
    simulationSpace=RootRelative
    alphaBoolBlend {blendOption=Linear}
    boundBone {boneName=dyn_hair_main_a_r}
    componentAppliedLinearAccClamp=<3500 3500 3500>
    componentLinearAccScale=<1.2 1.2 1.2>
    componentLinearVelScale=<1.2 1.2 1.2>
    physicsBodyDefinitions [{
      sphereCollisionRadius=0
      boundBone {boneName=dyn_hair_main_a_r}
      boxExtents=<20 10 10>
      constraintSetup {angularLimitsMax=<1 5 2>, angularLimitsMin=<-1 -2 -10>, angularTarget=<0 0 0>}
      localJointOffset=<2 0 0>
     }]
    relativeSpaceBone {boneName=root}
   }
   n20 = Anim Dynamics - Bone: dyn_hair_main_b_r(ComponentPose=@n19, Alpha=1, ExternalForce="0,0,0") {
    angularDampingOverride=0.9999
    angularSpringConstant=100
    bAngularSpring=true
    bGravityOverrideInSimSpace=true
    bOverrideAngularDamping=true
    simulationSpace=RootRelative
    alphaBoolBlend {blendOption=Linear}
    boundBone {boneName=dyn_hair_main_b_r}
    componentAppliedLinearAccClamp=<3500 3500 3500>
    componentLinearAccScale=<1.2 1.2 1.2>
    componentLinearVelScale=<1.2 1.2 1.2>
    physicsBodyDefinitions [{
      sphereCollisionRadius=0
      boundBone {boneName=dyn_hair_main_b_r}
      boxExtents=<20 10 10>
      constraintSetup {angularLimitsMax=<1 5 2>, angularLimitsMin=<-1 -2 -10>, angularTarget=<0 0 0>}
      localJointOffset=<2 0 0>
     }]
    relativeSpaceBone {boneName=root}
   }
   n21 = Anim Dynamics - Bone: dyn_hair_main_c_r(ComponentPose=@n20, Alpha=1, ExternalForce="0,0,0") {
    angularDampingOverride=0.9999
    angularSpringConstant=100
    bAngularSpring=true
    bGravityOverrideInSimSpace=true
    bOverrideAngularDamping=true
    simulationSpace=RootRelative
    alphaBoolBlend {blendOption=Linear}
    boundBone {boneName=dyn_hair_main_c_r}
    componentAppliedLinearAccClamp=<3500 3500 3500>
    componentLinearAccScale=<1.2 1.2 1.2>
    componentLinearVelScale=<1.2 1.2 1.2>
    physicsBodyDefinitions [{
      sphereCollisionRadius=0
      boundBone {boneName=dyn_hair_main_c_r}
      boxExtents=<20 10 10>
      constraintSetup {angularLimitsMax=<1 10 2>, angularLimitsMin=<-1 -2 -10>, angularTarget=<0 0 0>}
      localJointOffset=<2 0 0>
     }]
    relativeSpaceBone {boneName=root}
   }
   n22 = "Constraint - dyn_hair_main_b_l (P)"(ComponentPose=@n21, Alpha=0.4, ConstraintWeights_0=1) {
    alphaBoolBlend {blendOption=Linear}
    boneToModify {boneName=dyn_hair_main_b_l}
    constraintSetup [{targetBone {boneName=clavicle_l}}]
    constraintWeights [1]
   }
   n23 = "Constraint - dyn_hair_main_b_r (P)"(ComponentPose=@n22, Alpha=0.4, ConstraintWeights_0=1) {
    alphaBoolBlend {blendOption=Linear}
    boneToModify {boneName=dyn_hair_main_b_r}
    constraintSetup [{targetBone {boneName=clavicle_r}}]
    constraintWeights [1]
   }
   Output Pose(Result=Blend Poses by bool(BlendPose_0=Component To Local(ComponentPose=@n2), BlendPose_1=Component To Local(ComponentPose=@n23), bActiveValue=Get bRigidBodyEnabled()) {blendType=Linear}) {layerGroup=None}
   n24 = Copy Pose From Mesh(SourceMeshComponent=Get Body_Mesh_b() -> (Body_Mesh_b:Skeletal Mesh Component Object Reference))
   n25 = Copy Pose From Mesh() {bCopyCurves=true, bUseAttachedParent=true}
   n26 = Get HeadRotation() -> (HeadRotation:float)
   n27 = "Transform (Modify) Bone - Bone: dyn_hair_main_a_base"!(
    ComponentPose=Local To Component(LocalPose=Layered blend per bone(BasePose=@n25, BlendPoses_0=@n24) {layerSetup [{branchFilters [{boneName=upperarm_l}, {boneName=upperarm_r}]}]})
    Alpha=@n26
    Translation="0,0,0"
    Rotation="0,-30,0"
    Scale:Vector="1,1,1"
   ) -> (Pose) {
    rotationMode=BMM_Additive
    rotationSpace=BCS_BoneSpace
    translationMode=BMM_Additive
    translationSpace=BCS_BoneSpace
    alphaBoolBlend {blendOption=Linear}
    alphaScaleBiasClamp {bMapRange=true, inRange {max=0, min=-25}, outRange {max=0, min=1}}
    boneToModify {boneName=dyn_hair_main_a_base}
    rotation {yaw=-30}
   }
   n28 = "Transform (Modify) Bone - Bone: dyn_hair_main_a_root_base_l"!(ComponentPose=@n27, Alpha=@n26, Translation="0,0,0", Rotation="-30,0,0", Scale:Vector="1,1,1") -> (Pose) {
    rotationMode=BMM_Additive
    rotationSpace=BCS_BoneSpace
    translationMode=BMM_Additive
    translationSpace=BCS_BoneSpace
    alphaBoolBlend {blendOption=Linear}
    alphaScaleBiasClamp {bMapRange=true, inRange {max=0, min=-25}, outRange {max=0, min=1}}
    boneToModify {boneName=dyn_hair_main_a_root_base_l}
    rotation {pitch=-30}
   }
   n29 = "Transform (Modify) Bone - Bone: dyn_hair_main_a_base_root_base_r"!(ComponentPose=@n28, Alpha=@n26, Translation="0,0,0", Rotation="-30,0,0", Scale:Vector="1,1,1") -> (Pose) {
    rotationMode=BMM_Additive
    rotationSpace=BCS_BoneSpace
    translationMode=BMM_Additive
    translationSpace=BCS_BoneSpace
    alphaBoolBlend {blendOption=Linear}
    alphaScaleBiasClamp {bMapRange=true, inRange {max=0, min=-25}, outRange {max=0, min=1}}
    boneToModify {boneName=dyn_hair_main_a_base_root_base_r}
    rotation {pitch=-30}
   }
   n30 = Get HeadRotation()
   n31 = "Transform (Modify) Bone - Bone: dyn_hair_main_a_base"(ComponentPose=@n29, Alpha=@n30, Translation="0,0,0", Rotation="0,10,0", Scale="1,1,1") {
    rotationMode=BMM_Additive
    rotationSpace=BCS_BoneSpace
    translationMode=BMM_Additive
    translationSpace=BCS_BoneSpace
    alphaBoolBlend {blendOption=Linear}
    alphaScaleBiasClamp {bMapRange=true, inRange {max=20}}
    boneToModify {boneName=dyn_hair_main_a_base}
    rotation {yaw=10}
   }
   n32 = "Transform (Modify) Bone - Bone: dyn_hair_main_a_root_base_l"(ComponentPose=@n31, Alpha=@n30, Translation="0,0,0", Rotation="10,0,0", Scale="1,1,1") {
    rotationMode=BMM_Additive
    rotationSpace=BCS_BoneSpace
    translationMode=BMM_Additive
    translationSpace=BCS_BoneSpace
    alphaBoolBlend {blendOption=Linear}
    alphaScaleBiasClamp {bMapRange=true, inRange {max=20}}
    boneToModify {boneName=dyn_hair_main_a_root_base_l}
    rotation {pitch=10}
   }
   n33 = "Transform (Modify) Bone - Bone: dyn_hair_main_a_base_root_base_r"(ComponentPose=@n32, Alpha=@n30, Translation="0,0,0", Rotation="10,0,0", Scale="1,1,1") {
    rotationMode=BMM_Additive
    rotationSpace=BCS_BoneSpace
    translationMode=BMM_Additive
    translationSpace=BCS_BoneSpace
    alphaBoolBlend {blendOption=Linear}
    alphaScaleBiasClamp {bMapRange=true, inRange {max=20}}
    boneToModify {boneName=dyn_hair_main_a_base_root_base_r}
    rotation {pitch=10}
   }
   Save cached pose 'Sim'(Pose=Component To Local(ComponentPose=@n33)) {cacheName=Sim}
  ]}
 EventGraph {nodes [
   Event BlueprintInitializeAnimation()
   > Set Body_Mesh_b(Body_Mesh_b:Skeletal Mesh Component Object Reference=GetOwnerSkeletalMeshComponent()) -> (Output_Get:Skeletal Mesh Component Object Reference)
  ]}
]
variables [Body_Mesh_b {type=SkeletalMeshComponent}]
